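# Animation Blueprint for the Going Commando soldier minigun gadget that spins the barrel bone according to the weapon's charge time.
# Each update reads charge state from the owning B_Ranged_Generic, following charge time while charging and for one second afterward, then easing it to zero.
# That value, scaled by max speed and delta time, accumulates into a roll.
# The AnimGraph adds the roll to Mini_Gun_Anim over a static, yaw-rotated weapon pose, then passes through DefaultSlot.
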AnimBlueprint /Game/Animation/Game/MainPlayer/Combat/Gadgets/Medium/BP_Soldier_Minigun
target_skeleton=/Game/Gadgets/GoingCommando/Mesh/SK_M_Med_Soldier_04_Minigun_Skeleton
graphs [
 AnimGraph {nodes [
   n1 = "Sequence Evaluator (by time) 'M_GoingCommando_Weapon'"!(ExplicitTime:float=0) -> (Pose:Pose Link Structure) {is=title AnimGraphNode_SequenceEvaluator, sequence=/Game/Animation/Game/MainPlayer/Poses/Medium/Male/M_GoingCommando_Weapon}
   n2 = "Transform (Modify) Bone - Bone: Mini_Gun_Anim"(
    ComponentPose=Local To Component(LocalPose=Rotate Root Bone!(BasePose:Pose Link Structure=@n1, Pitch:float=0, Yaw:float=90, MeshToComponent:Rotator="0,0,0") -> (Pose:Pose Link Structure) {is=title AnimGraphNode_RotateRootBone})
    Alpha=1
    Translation="0,0,0"
    Rotation=MakeRotator(Roll=Get MinigunRotation() -> (MinigunRotation:double))
    Scale="1,1,1"
   ) {rotationMode=BMM_Additive, rotationSpace=BCS_BoneSpace, alphaBoolBlend {blendOption=Linear}, boneToModify {boneName=Mini_Gun_Anim}}
   Output Pose(Result=Slot 'DefaultSlot'(Source=Component To Local(ComponentPose=@n2))) {layerGroup=None}
  ]}
 EventGraph {nodes [
   n1 = Event BlueprintUpdateAnimation()
   > Set CurDeltaTime(CurDeltaTime:double=@n1.DeltaTimeX) -> (Output_Get:double)
   > n2 = Cast To B_Ranged_Generic!(execute, Object=GetOwningActor()) -> (then, CastFailed, AsB Ranged Generic:B Ranged Generic Object Reference) {is=title K2Node_DynamicCast}
   > Set CurChargeTime(CurChargeTime:double=GetChargeTime!(self:Fort Weapon Object Reference=@n2) -> (ReturnValue:float)) -> (Output_Get:double)
   > n3 = Branch(Condition=GetIsChargingWeapon!(self:Fort Weapon Object Reference=@n2) -> (ReturnValue:bool))
   > n4 = Set LastChargingTime(LastChargingTime:double=GetTimeSeconds!() -> (ReturnValue:double)) -> (Output_Get:double)
   n5 = Set ChargeTimeOverride(execute=[@n11.then, @n4.then], ChargeTimeOverride:double=Get CurChargeTime() -> (CurChargeTime:double)) -> (Output_Get:double)
   n6 = Get MaxMinigunRotSpeed() -> (MaxMinigunRotSpeed:double)
   n7 = Get ChargeTimeOverride() -> (ChargeTimeOverride:double)
   n8 = Get CurDeltaTime() -> (CurDeltaTime:double)
   Set MinigunRotation(
    execute=[@n5.then, @n12.then]
    MinigunRotation:double=float + float(A=float * float!(A=@n7, B=@n6, C:double=@n8) -> (ReturnValue), B=Get MinigunRotation()) {is=title K2Node_CommutativeAssociativeBinaryOperator}
   ) -> (Output_Get:double)
   n9 = GetTimeSeconds()
   n10 = Get LastChargingTime() -> (LastChargingTime:double)
   n11 = Branch(
    execute=@n3.else
    Condition="float <= float"!(A:double=float - float(A=@n9, B=@n10) {is=title K2Node_CallFunction}, B:double=1) -> (ReturnValue:bool) {is=title K2Node_CallFunction}
   )
   n12 = Set ChargeTimeOverride(execute=@n11.else, ChargeTimeOverride=FInterpTo(Current=Get ChargeTimeOverride(), Target=0, DeltaTime=Get CurDeltaTime(), InterpSpeed=4))
  ]}
]
variables [
 ChargeTimeOverride {type=number}
 CurChargeTime {type=number}
 CurDeltaTime {type=number}
 LastChargingTime {type=number}
 MaxMinigunRotSpeed {type=number, default=1100}
 MinigunRotation {type=number}
]
